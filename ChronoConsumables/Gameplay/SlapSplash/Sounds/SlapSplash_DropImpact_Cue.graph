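# SoundCue for the impact sound when a SlapSplash consumable is dropped.
# A random node picks one of three equally weighted drop impact waves, and a modulator lowers its pitch within a small random range.
# It plays quietly in the gadgets and items sound class with weapon drop attenuation.
# A concurrency override caps simultaneous instances and blocks new plays once the limit is reached.
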SoundCue /ChronoConsumables/Gameplay/SlapSplash/Sounds/SlapSplash_DropImpact_Cue
root=Modulator_0
nodes [
 {id=Modulator_0, children [Random_0], properties {pitchMax=0.9, pitchMin=0.8}}
 {id=Random_0, children [WavePlayer_1,WavePlayer_2,WavePlayer_3], properties {weights [1,1,1]}}
 {id=WavePlayer_1, properties {soundWaveAssetPtr=&/SlapSplash_Drop_Impact_01}}
 {id=WavePlayer_2, properties {soundWaveAssetPtr=&/SlapSplash_Drop_Impact_02}}
 {id=WavePlayer_3, properties {soundWaveAssetPtr=&/SlapSplash_Drop_Impact_03}}
]
properties {
 attenuationSettings=&Fort_Weapon_Drop_Att
 bOverrideConcurrency=true
 maxDistance=3000
 soundClassObject=&Fort_Gadgets_Items
 volumeMultiplier=0.3
 concurrencyOverrides {maxCount=3, resolutionRule=PreventNew}
}
