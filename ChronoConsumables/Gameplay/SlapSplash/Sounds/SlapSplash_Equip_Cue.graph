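# Sound cue for equipping the SlapSplash consumable, routed through the player foley sound class with close-range prop attenuation.
# A Random node picks one of three equip waves with equal weight and feeds a Modulator at default settings for slight pitch and volume variation.
# Its own concurrency override prevents new instances from starting once the limit is reached.
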SoundCue /ChronoConsumables/Gameplay/SlapSplash/Sounds/SlapSplash_Equip_Cue
root=Modulator_0
nodes [
 {id=Modulator_0, children [Random_0]}
 {id=Random_0, children [WavePlayer_0,WavePlayer_2,WavePlayer_3], properties {weights [1,1,1]}}
 {id=WavePlayer_0, properties {soundWaveAssetPtr=&/SlapSplash_Equip_01}}
 {id=WavePlayer_2, properties {soundWaveAssetPtr=&/SlapSplash_Equip_02}}
 {id=WavePlayer_3, properties {soundWaveAssetPtr=&/SlapSplash_Equip_03}}
]
properties {
 attenuationSettings=/Game/Sounds/Attenuation/Fort_Prop_Close_Att
 bOverrideConcurrency=true
 maxDistance=1800
 soundClassObject=&Fort_Player_Foley
 volumeMultiplier=0.8
 concurrencyOverrides {resolutionRule=PreventNew}
}
