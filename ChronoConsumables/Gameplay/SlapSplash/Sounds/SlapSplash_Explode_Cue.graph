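# Sound cue for the SlapSplash explosion, routed to the large explosives sound class.
# A mixer blends a randomly chosen main explosion with a quieter, briefly delayed random explosion layer, each drawn from four waves with slight pitch and volume modulation.
# It uses the shared close-projectile attenuation asset and overrides concurrency to block new instances once the limit is reached.
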SoundCue /ChronoConsumables/Gameplay/SlapSplash/Sounds/SlapSplash_Explode_Cue
root=Mixer_0
nodes [
 {id=Delay_0, children [Random_2], properties {delayMax=0.1, delayMin=0.1}}
 {id=Mixer_0, children [Modulator_0,Modulator_1], properties {inputVolume [1,0.4]}}
 {id=Modulator_0, children [Random_0], properties {pitchMax=1.1, pitchMin=0.9, volumeMax=1.1, volumeMin=0.9}}
 {id=Modulator_1, children [Delay_0], properties {pitchMax=1.1, pitchMin=1, volumeMax=1.1, volumeMin=0.9}}
 {id=Random_0, children [WavePlayer_1,WavePlayer_2,WavePlayer_3,WavePlayer_4], properties {weights [1,1,1,1]}}
 {id=Random_2, children [WavePlayer_8, WavePlayer_9, WavePlayer_10, WavePlayer_11], properties {weights [1,1,1,1]}}
 {id=WavePlayer_1, properties {soundWaveAssetPtr=&/SlapSplash_Explode_01}}
 {id=WavePlayer_10, properties {soundWaveAssetPtr=&/SlapSplash_ExplosionLayer_03}}
 {id=WavePlayer_11, properties {soundWaveAssetPtr=&/SlapSplash_ExplosionLayer_04}}
 {id=WavePlayer_2, properties {soundWaveAssetPtr=&/SlapSplash_Explode_02}}
 {id=WavePlayer_3, properties {soundWaveAssetPtr=&/SlapSplash_Explode_03}}
 {id=WavePlayer_4, properties {soundWaveAssetPtr=&/SlapSplash_Explode_04}}
 {id=WavePlayer_8, properties {soundWaveAssetPtr=&/SlapSplash_ExplosionLayer_01}}
 {id=WavePlayer_9, properties {soundWaveAssetPtr=&/SlapSplash_ExplosionLayer_02}}
]
properties {
 attenuationSettings=&Fort_Weapon_Projectile_Close_Athena_Att
 bOverrideConcurrency=true
 maxDistance=4950
 soundClassObject=/Game/Sounds/FortSoundClassesAndModes/Fort_Explosives_Large_PSM
 volumeMultiplier=0.3184
 attenuationOverrides {
  bAttenuateWithLPF=true
  falloffDistance=4300
  lPFFrequencyAtMax=1000
  lPFRadiusMax=3200
  lPFRadiusMin=700
  stereoSpread=1200
  attenuationShapeExtents {x=300}
 }
 concurrencyOverrides {resolutionRule=PreventNew}
}
