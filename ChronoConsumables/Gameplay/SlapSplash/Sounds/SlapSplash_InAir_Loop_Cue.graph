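# Sound cue for the SlapSplash consumable's in-air loop, judging by its name and projectile attenuation.
# A single looping wave player feeds a modulator that pins pitch slightly below normal with no random range.
# The cue plays at reduced volume through Fortnite's shared projectile in-air attenuation and a third-person ranged weapon fire sound class.
# A concurrency override caps it at a few simultaneous instances.
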SoundCue /ChronoConsumables/Gameplay/SlapSplash/Sounds/SlapSplash_InAir_Loop_Cue
root=Modulator_0
nodes [
 {id=Modulator_0, children [WavePlayer_1], properties {pitchMax=0.9, pitchMin=0.9}}
 {id=WavePlayer_1, properties {bLooping=true, soundWaveAssetPtr=&/SlapSplash_InAir_Loop_01}}
]
properties {
 attenuationSettings=/Game/Sounds/Attenuation/Fort_Weapon_Projectile_InAir_Loop_Att
 bOverrideConcurrency=true
 maxDistance=4500
 soundClassObject=&Fort_WeaponRanged_Fire_3P_PSM
 volumeMultiplier=0.4
 concurrencyOverrides {maxCount=3}
}
