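# Sound cue for picking up the SlapSplash consumable.
# It randomly plays one of two pickup waves with equal weight, then modulates pitch slightly downward and varies volume a little.
# It plays quietly through the gadgets-items sound class with weapon-pickup attenuation.
# A concurrency override prevents new instances from starting once its limit is reached.
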SoundCue /ChronoConsumables/Gameplay/SlapSplash/Sounds/SlapSplash_Pickup_Cue
root=Modulator_0
nodes [
 {id=Modulator_0, children [Random_0], properties {pitchMax=0.9, pitchMin=0.8, volumeMax=1.1, volumeMin=0.9}}
 {id=Random_0, children [WavePlayer_1,WavePlayer_3], properties {weights [1,1]}}
 {id=WavePlayer_1, properties {soundWaveAssetPtr=&/SlapSplash_Pickup_01}}
 {id=WavePlayer_3, properties {soundWaveAssetPtr=&/SlapSplash_Pickup_03}}
]
properties {
 attenuationSettings=&Fort_Weapon_Pickup_Att
 bOverrideConcurrency=true
 maxDistance=3500
 soundClassObject=&Fort_Gadgets_Items
 volumeMultiplier=0.25
 concurrencyOverrides {resolutionRule=PreventNew}
}
