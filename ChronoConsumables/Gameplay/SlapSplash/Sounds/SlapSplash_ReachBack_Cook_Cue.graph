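# Sound cue for the SlapSplash consumable's reach-back cook sound.
# A Random node picks one of three equally weighted reach-back waves, which feeds a Modulator that lowers pitch with slight random variation.
# It plays quietly through the gadgets sound class, using weapon-reload attenuation for positional falloff.
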SoundCue /ChronoConsumables/Gameplay/SlapSplash/Sounds/SlapSplash_ReachBack_Cook_Cue
root=Modulator_1
nodes [
 {id=Modulator_1, children [Random_0], properties {pitchMax=0.95, pitchMin=0.85}}
 {id=Random_0, children [WavePlayer_0,WavePlayer_1,WavePlayer_2], properties {weights [1,1,1]}}
 {id=WavePlayer_0, properties {soundWaveAssetPtr=&/SlapSplash_ReachBack_01}}
 {id=WavePlayer_1, properties {soundWaveAssetPtr=&/SlapSplash_ReachBack_02}}
 {id=WavePlayer_2, properties {soundWaveAssetPtr=&/SlapSplash_ReachBack_03}}
]
properties {
 attenuationSettings=&Fort_Weapon_Reload_Att
 maxDistance=2250
 soundClassObject=&Fort_Gadgets
 volumeMultiplier=0.17
}
