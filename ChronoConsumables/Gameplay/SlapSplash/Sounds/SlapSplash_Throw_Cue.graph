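# Sound cue for the throw sound of the SlapSplash consumable gadget.
# A Random node picks one of six equally weighted throw wave variations, then a Modulator adds slight random volume variation and upward-only pitch variation.
# The cue plays at reduced volume through the Fort_Gadgets sound class, spatialized with the SlapSplash gadget attenuation asset.
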SoundCue /ChronoConsumables/Gameplay/SlapSplash/Sounds/SlapSplash_Throw_Cue
root=Modulator_0
nodes [
 {id=Modulator_0, children [Random_1], properties {pitchMax=1.1, pitchMin=1, volumeMax=1.1, volumeMin=0.9}}
 {
  id=Random_1
  children [WavePlayer_2, WavePlayer_6, WavePlayer_7, WavePlayer_8, WavePlayer_9, WavePlayer_10]
  properties {weights [1,1,1,1,1,1]}
 }
 {id=WavePlayer_10, properties {soundWaveAssetPtr=&/SlapSplash_Throw_06}}
 {id=WavePlayer_2, properties {soundWaveAssetPtr=&/SlapSplash_Throw_01}}
 {id=WavePlayer_6, properties {soundWaveAssetPtr=&/SlapSplash_Throw_02}}
 {id=WavePlayer_7, properties {soundWaveAssetPtr=&/SlapSplash_Throw_03}}
 {id=WavePlayer_8, properties {soundWaveAssetPtr=&/SlapSplash_Throw_04}}
 {id=WavePlayer_9, properties {soundWaveAssetPtr=&/SlapSplash_Throw_05}}
]
properties {
 attenuationSettings=&/SlapSplash_Gadget_Att
 maxDistance=2000
 soundClassObject=&Fort_Gadgets
 volumeMultiplier=0.5
}
