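# Sound cue for the SlapSplash consumable that plays a whip effect, classed as a ranged weapon projectile effect.
# It picks one of four equally weighted whip waves at random, then a modulator slightly varies pitch and volume on each play.
# Attenuation overrides take the place of the assigned melee projectile whip attenuation asset, disabling reverb send and using complex collision for occlusion.
# The cue plays at reduced volume and has its debug flag on.
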SoundCue /ChronoConsumables/Gameplay/SlapSplash/Sounds/SlapSplash_Whip_Cue
root=Modulator_2
nodes [
 {id=Modulator_2, children [Random_4], properties {pitchMax=1.1, pitchMin=0.9, volumeMax=1.1, volumeMin=0.9}}
 {id=Random_4, children [WavePlayer_12, WavePlayer_13, WavePlayer_14, WavePlayer_15], properties {weights [1,1,1,1]}}
 {id=WavePlayer_12, properties {soundWaveAssetPtr=&/SlapSplash_Whip_01}}
 {id=WavePlayer_13, properties {soundWaveAssetPtr=&/SlapSplash_Whip_02}}
 {id=WavePlayer_14, properties {soundWaveAssetPtr=&/SlapSplash_Whip_03}}
 {id=WavePlayer_15, properties {soundWaveAssetPtr=&/SlapSplash_Whip_04}}
]
properties {
 attenuationSettings=/Game/Sounds/Attenuation/Fort_Weapon_Melee_Lobster_Prj_InAir_Whip_Att
 bDebug=true
 bOverrideAttenuation=true
 maxDistance=1512
 soundClassObject=&Fort_WeaponRanged_Projectile_Effects
 volumeMultiplier=0.6
 attenuationOverrides {
  bEnableReverbSend=false
  bUseComplexCollisionForOcclusion=true
  falloffDistance=1000
  occlusionVolumeAttenuation=0
  stereoSpread=0
  attenuationShapeExtents {x=512}
 }
}
